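# Material function for a Time Trial dissolve effect, outputting OpacityMask, Emissive and Opacity.
# Timestamp, length and reverse-direction custom primitive data drive a smoothstepped transition that grows a sphere mask along a GradientDot-oriented UV1 gradient.
# Time-animated random UV noise, passed through a sine, is added to that mask to break up the dissolve.
# Emissive boosts eroded areas by inverting the mask, and Opacity multiplies it by transition progress; debug time nodes are unconnected.
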MaterialFunction /Game/Effects/Fort_Effects/Materials/UI/MF_TimeTrial_Dissolve
graph {nodes [
  n1 = Divide(A=Time(), B=ScalarParameter() {defaultValue=2, parameterName=debugSpeed})
  n2 = Constant() {r=1}
  ConstantBiasScale(None=OneMinus(None=Fmod(A=@n1, B=@n2)))
  n3 = TextureCoordinate() {coordinateIndex=1}
  n4 = ScalarParameter() {defaultValue=22, parameterName=UV}
  n5 = Add(
   A=Fmod(A=Time(), B=Constant() {r=1})
   B=MaterialFunctionCall!(PrimaryTextureCoordinates=Multiply(A=@n3, B=@n4), 2ndOrderTextureCoordinates, RandomSeed, NoiseDensityBias, Bypass Noise Influence Crosstalk) {materialFunction=/Game/Effects/Fort_Effects/Materials/_Functions/MF_RandomUVSample}.NoiseWithRange
  )
  n6 = TextureCoordinate() {coordinateIndex=1}
  n7 = FunctionInput(Preview=Constant2Vector() {g=1}) {bUsePreviewValueAsDefault=true, inputName=GradientDot, inputType=FunctionInput_Vector2}
  n8 = MaterialFunctionCall(
   TimestampA=ScalarParameter() {bUseCustomPrimitiveData=true, defaultValue=99999997952, parameterName=Timestamp}
   TransitionLength_In=ScalarParameter() {bUseCustomPrimitiveData=true, defaultValue=1, parameterName=TransitionLength, primitiveDataIndex=1}
  ) {materialFunction=&MF_Timestamp_InOut_01}
  n9 = LinearInterpolate(
   A=@n8.SmoothStep
   B=OneMinus(None=@n8.SmoothStep)
   Alpha=ScalarParameter() {bUseCustomPrimitiveData=true, parameterName=ReverseDirection, primitiveDataIndex=3}
  )
  n10 = SphereMask(
   A=SphereMask(A=DotProduct(A=@n6, B=@n7), B=Constant() {r=0.5}, Radius=Constant() {r=1}, Hardness=Constant())
   B=Constant()
   Radius=@n9
   Hardness=Power(Base=@n9, Exp=ScalarParameter() {defaultValue=5, parameterName=HardnessExponent})
  )
  n11 = Add(
   A=ConstantBiasScale(None=Sine(None=@n5))
   B=Multiply(A=Subtract(A=@n10) {constB=0.5}, B=ScalarParameter() {defaultValue=3, parameterName=TransitionScale})
  )
  FunctionOutput(None=@n11) {outputName=OpacityMask}
  n12 = OneMinus(None=@n11)
  n13 = ScalarParameter() {defaultValue=200, parameterName=ColorErodeScale}
  FunctionOutput(None=Clamp(None=Multiply(A=@n12, B=@n13)) {clampMode=CMODE_ClampMin, minDefault=1}) {outputName=Emissive, sortPriority=1}
  FunctionOutput(None=Multiply(A=@n11, B=@n9)) {outputName=Opacity}
  ScalarParameter() {defaultValue=1, parameterName=DebugTime}
 ]}
