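# Unlit masked material for the edge of a time trial marker, with sine bounce and optional camera-facing vertex offset.
# FaceTowardsCamera also switches the color from a Fresnel two-color blend to flat ColorA.
# MF_TimeTrial_Dissolve boosts emissive and supplies opacity, which blinks once TimeLeft drops to a quarter or below.
# A camera-distance scale-up branch exists but does not feed the bounce and facing offset.
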Material /Game/Effects/Fort_Effects/Materials/UI/M_TimeTrial_Marker_Edge
graph {
 properties {blendMode=BLEND_Masked, opacityMaskClipValue=0.5, pixelDepthOffsetMode=PDOM_Legacy, shadingModel=MSM_Unlit, usedShadingModels=Unlit}
 nodes [
  n1 = Sine(None=Multiply(A=Time(), B=ScalarParameter() {defaultValue=0.75, parameterName=BounceSpeed}))
  n2 = Constant3Vector() {constant=<rgba 0 0 1 0>}
  n3 = Normalize(VectorInput=Transform(None=Constant3Vector() {constant=<rgba 0 0 1 0>}) {transformSourceType=TRANSFORMSOURCE_Local})
  n4 = Normalize(VectorInput=Transform(None=Constant3Vector() {constant=<rgba 0 1 0 0>}) {transformSourceType=TRANSFORMSOURCE_Local})
  n5 = Normalize(VectorInput=Transform(None=Constant3Vector() {constant=<rgba 1 0 0 0>}) {transformSourceType=TRANSFORMSOURCE_Local})
  n6 = ScalarParameter() {defaultValue=1, parameterName=FaceTowardsCamera}
  Add(
   A=Multiply(A=Multiply(A=@n1, B=@n2), B=ScalarParameter() {defaultValue=6, parameterName=BounceAmount})
   B=LinearInterpolate(
    B=MaterialFunctionCall!(WorldPosition, PivotLocation, WS Normals, Custom Object Basis 1=@n3, Custom Object Basis 2=@n4, Custom Object Basis 3=@n5) {materialFunction=/Engine/Functions/Engine_MaterialFunctions02/WorldPositionOffset/AlignMeshToTheCamera}.World Position Offset
    Alpha=@n6
   )
  )
  n7 = VectorParameter() {parameterName=ColorA, defaultValue=<rgba 0.1562 2.2798 3 1>}
  n8 = VectorParameter() {parameterName=ColorB, defaultValue=<rgba 0.025 0.0373 0.0521 1>}
  n9 = Fresnel(Normal=VertexNormalWS()) {exponent=1}
  n10 = Multiply(
   A=ScalarParameter() {defaultValue=1, parameterName=Brightness}
   B=LinearInterpolate(A=LinearInterpolate(A=@n7.RGB, B=@n8.RGB, Alpha=@n9), B=@n7.RGB, Alpha=@n6)
  )
  n11 = MaterialFunctionCall!(GradientDot=Constant2Vector() {r=1}) {materialFunction=&/MF_TimeTrial_Dissolve}
  Add(A=@n10, B=Multiply(A=@n10, B=@n11.Emissive))
  n12 = ScalarParameter() {defaultValue=50, parameterName=MaxScaleIncrease}
  n13 = ObjectPositionWS()
  n14 = WorldPosition()
  n15 = Subtract(A=Distance(A=@n13, B=CameraPositionWS()), B=ScalarParameter() {defaultValue=500, parameterName=DistanceOffset})
  n16 = ScalarParameter() {defaultValue=30000, parameterName=ScaleUpDistance}
  LinearInterpolate(B=Multiply(A=Multiply(A=@n12) {constB=-1}, B=Subtract(A=@n13, B=@n14.XYZ)) {constB=-1}, Alpha=Clamp(None=Divide(A=@n15, B=@n16))) {constB=0}
  MaterialFunctionCall(Speed=Constant() {r=0.2}) {materialFunction=&DebugTimeSine}
  n17 = Multiply(A=Time(), B=ScalarParameter() {defaultValue=3, parameterName=BlinkSpeed})
  n18 = Round(None=ConstantBiasScale(None=Sine(None=@n17)))
  Multiply(
   A=If(A=ScalarParameter() {defaultValue=1, parameterName=TimeLeft}, B=Constant() {r=0.25}, "A > B"=Constant() {r=1}, "A == B"=@n18, "A < B"=@n18)
   B=@n11.OpacityMask
  )
 ]
}
